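; Looping gameplay cue for the Sulfer damage-over-time effect that plays a 2D storm zap sound when the player pawn's health changes.
; On loop start it binds to the pawn's health-changed and end-skydiving events, and on removal it unbinds the health event.
; The skydiving event sets a flag that gates the sound, then unbinds itself.
; A DoOnce guards playback and starts a delay task, with no reset shown.
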(event EventOnLoopingStart(Cascade)
  (params (MyTarget :type Actor%20Object%20Reference) (Parameters :type Gameplay%20Cue%20Parameters%20Structure%20%28by%20ref%29) (ParticleComponents :type Array%20of%20Cascade%20Particle%20System%20Component%20Object%20References) (AudioComponents :type Array%20of%20Audio%20Component%20Object%20References))
  (bind _asplayer_pawn_athena (Utilities|Casting|CastToPlayerPawn_Athena MyTarget))
  (bind _output_get (Variables|Default|SetPlayerPawn _asplayer_pawn_athena))
  (Health|BindEventtoOnPawnHealthChanged _output_get (EventDispatchers|CreateEvent :!event Custom|Pawn%20Health%20Changed))
  (EventDispatchers|BindEventtoOnEndSkydivingEvent _output_get (EventDispatchers|CreateEvent :!event Custom|Pawn%20Finished%20Skydiving)))

(event Custom|Pawn%20Health%20Changed
  ;[0 0.09724 1] Play a Zap Sound when Pawn is damaged
  ;;[1 1 1] Prevents sound from playing before skydiving has finished
  (if (Variables|Default|GetHas%20Pawn%20Finished%20Skydiving)
    (bind _playerpawn (Variables|Default|GetPlayerPawn)
      (:then
        ;;[1 1 1] Play SFX
        (Utilities|FlowControl|DoOnce false
          (:Completed
            (Audio|PlaySound2D _playerpawn "/FNGameplayCues/Athena/Sounds/Impacts/Storm_Player_Damage_Zaps_Cue.Storm_Player_Damage_Zaps_Cue" 1.0 1.0 0.0 0 _playerpawn)
            (GameplayTasks|TaskWaitDelay (Ability|GetAbilitySystemComponent (Game|GetInstigator _playerpawn)) (Variables|Default|GetDelaySoundVariable)
              (:OnFinish)))))
      (:else))))

(event EventOnRemoval(Cascade)
  (params (MyTarget :type Actor%20Object%20Reference) (Parameters :type Gameplay%20Cue%20Parameters%20Structure%20%28by%20ref%29) (ParticleComponents :type Array%20of%20Cascade%20Particle%20System%20Component%20Object%20References) (AudioComponents :type Array%20of%20Audio%20Component%20Object%20References) (BurstCameraShakeInstance :type Camera%20Shake%20Base%20Object%20Reference) (BurstDecalInstance :type Decal%20Actor%20Object%20Reference))
  (bind _playerpawn (Variables|Default|GetPlayerPawn)
    (:then
      (Health|UnbindEventfromOnPawnHealthChanged _playerpawn (EventDispatchers|CreateEvent :!event Custom|Pawn%20Health%20Changed)))
    (:else)))

(event Custom|Pawn%20Finished%20Skydiving
  ;[1 1 1] Unbind 'On End Skydiving Event' once the function is called
  (bind _playerpawn (Variables|Default|GetPlayerPawn)
    (:then
      (EventDispatchers|UnbindEventfromOnEndSkydivingEvent _playerpawn (EventDispatchers|CreateEvent :!event Custom|Pawn%20Finished%20Skydiving))
      (Variables|Default|SetHas%20Pawn%20Finished%20Skydiving true))
    (:else)))

(fn ConstructionScript)
