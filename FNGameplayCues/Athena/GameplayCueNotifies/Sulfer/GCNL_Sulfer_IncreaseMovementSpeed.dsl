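; Looping gameplay cue for the Sulfer LTM movement speed increase that drives a directional sprint wind sound on the player pawn.
; It binds to the pawn's movement update on loop start, unbinds on removal, and uses DoOnce gates to spawn the attached sound with a timeline volume ramp and fade it out when swimming or slow.
; Interpolated movement direction relative to facing is normalized into front, back, left and right float parameters on the audio component.
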(event EventOnLoopingStart(Cascade)
  (params (MyTarget :type Actor%20Object%20Reference) (Parameters :type Gameplay%20Cue%20Parameters%20Structure%20%28by%20ref%29) (ParticleComponents :type Array%20of%20Cascade%20Particle%20System%20Component%20Object%20References) (AudioComponents :type Array%20of%20Audio%20Component%20Object%20References))
  (bind _asplayer_pawn_athena (Utilities|Casting|CastToPlayerPawn_Athena MyTarget))
  (bind _output_get (Variables|Default|SetPlayerPawn _asplayer_pawn_athena))
  (Character|BindEventtoOnCharacterMovementUpdated _output_get (EventDispatchers|CreateEvent :!event Custom|PawnMovement)))

(event EventOnRemoval(Cascade)
  (params (MyTarget :type Actor%20Object%20Reference) (Parameters :type Gameplay%20Cue%20Parameters%20Structure%20%28by%20ref%29) (ParticleComponents :type Array%20of%20Cascade%20Particle%20System%20Component%20Object%20References) (AudioComponents :type Array%20of%20Audio%20Component%20Object%20References) (BurstCameraShakeInstance :type Camera%20Shake%20Base%20Object%20Reference) (BurstDecalInstance :type Decal%20Actor%20Object%20Reference))
  (bind _playerpawn (Variables|Default|GetPlayerPawn)
    (:then
      (Character|UnbindEventfromOnCharacterMovementUpdated _playerpawn (EventDispatchers|CreateEvent :!event Custom|PawnMovement)))
    (:else)))

(event Custom|PawnMovement
  (params (DeltaSeconds :type Float%20%28single-precision%29) (OldLocation :type Vector) (OldVelocity :type Vector))
  ;[0 0.153 1] Sprinting SFX
  (Variables|Default|SetDeltaSeconds DeltaSeconds)
  (bind _playerpawn (Variables|Default|GetPlayerPawn)
    (:then
      ;;[1 1 1] Is Pawn Swimming?
      (if (WaterBody|IsSurfaceSwimming _playerpawn)
        ;;[1 1 1] Play Air Sprinting SFX
        ;;[1 1 1] Stop SFX
        (Utilities|FlowControl|Sequence :!label _label_3
          (:then_0
            (Utilities|FlowControl|DoOnce false :!label _label
              (:Completed
                (Audio|Components|Audio|FadeOut (Variables|Default|GetSprintSFX) (Variables|Default|GetFadeOutDurationforWindSFX) 0.0 "Logarithmic" :!class /Script/Engine.AudioComponent)
                (bind (_direction _volume) (Timeline|VolumeRampUp Stop)
                  (:Update
                    (Audio|Components|Audio|AdjustVolume (Variables|Default|GetSprintSFX) 0.0 _volume "Linear" :!class /Script/Engine.AudioComponent))))))
          (:then_1
            _)
          (:then_2
            ;;[1 1 1] Find Which Left/Right direction Pawn is going
            ;;;[1 1 1] Interpolate direction for smooth transition.
            (bind _output_get (Variables|Default|SetDirection_Target_LeftRight (Animation|CalculateDirection (Components|SkeletalMesh|GetAnimInstance (Variables|/Script/Engine.Character|GetMesh _playerpawn)) (Pawn|Components|CharacterMovement|GetLastUpdateVelocity (Variables|/Script/Engine.Character|GetCharacterMovement _playerpawn)) (Transformation|GetActorRotation _playerpawn)) :!label _label_2))
            (bind (_returnvalue_roll _returnvalue_pitch _returnvalue_yaw) (Math|Rotator|BreakRotator (Math|Interpolation|RInterpTo (Math|Rotator|MakeRotator 0.0 0.0 (Variables|Default|GetDirection_Interp_LeftRight)) (Math|Rotator|MakeRotator 0.0 0.0 _output_get) (Variables|Default|GetDeltaSeconds) (Variables|Default|GetInterpolated%20Speed))))
            (Variables|Default|SetDirection_Interp_LeftRight _returnvalue_yaw)
            ;;[1 1 1] Set Direction Pawn is Running For Sound
            (CallFunction|ConvertMin/MaxIntoZerotoOne (Variables|Default|GetDirection_Interp_LeftRight) (Variables|Default|GetFront_Min) (Variables|Default|GetFront_Max_-_Min) (Variables|Default|GetSprintSFX) (Variables|Default|GetFrontParam))
            (CallFunction|ConvertMin/MaxIntoZerotoOne (Variables|Default|GetDirection_Interp_LeftRight) (Variables|Default|GetLeft_Min) (Variables|Default|GetLeft_Max_-_Min) (Variables|Default|GetSprintSFX) (Variables|Default|GetLeftParam))
            (CallFunction|ConvertMin/MaxIntoZerotoOne (Variables|Default|GetDirection_Interp_LeftRight) (Variables|Default|GetRight_Min) (Variables|Default|GetRight_Max_-_Min) (Variables|Default|GetSprintSFX) (Variables|Default|GetRightParam))
            (if (< (Variables|Default|GetDirection_Interp_LeftRight) (Variables|Default|GetBack_Min))
              (CallFunction|ConvertMin/MaxIntoZerotoOne (Variables|Default|GetDirection_Interp_LeftRight) (Variables|Default|GetBack_Min) (Variables|Default|GetBack_Max_-_Min) (Variables|Default|GetSprintSFX) (Variables|Default|GetBackParam))
              (else
                (CallFunction|ConvertMin/MaxIntoZerotoOne (Variables|Default|GetDirection_Interp_LeftRight) (Variables|Default|GetBack2_Min) (Variables|Default|GetBack2_Max_-_Min) (Variables|Default|GetSprintSFX) (Variables|Default|GetBackParam))))))
        (elif (>= (Math|Vector|VectorLength (Pawn|Components|CharacterMovement|GetLastUpdateVelocity (Variables|/Script/Engine.Character|GetCharacterMovement _playerpawn))) (Variables|Default|GetVelocityForWindSFX))
          (Utilities|FlowControl|Sequence
            (:then_0
              _)
            (:then_1
              (goto _label Reset))
            (:then_2
              (goto _label_2)))
          (else
            (goto _label_3))))
      (Utilities|FlowControl|DoOnce false
        (:Completed
          (bind _returnvalue (Audio|SpawnSoundAttached :Sound "/FNGameplayCues/Sounds/LTM/Sulfer/Sulfer_Sprint_Cue.Sulfer_Sprint_Cue" :AttachToComponent (Variables|Transformation|GetRootComponent _playerpawn) :AttachPointName "None" :LocationType "KeepRelativeOffset" :bStopWhenAttachedToDestroyed false :VolumeMultiplier 0.5 :PitchMultiplier 1.0 :StartTime 0.0 :bAutoDestroy true))
          (Variables|Default|SetSprintSFX _returnvalue)
          (Timeline|VolumeRampUp PlayFromStart))))
    (:else)))

(fn Convert_Min/Max_into_Zero_To_One
  (params (Input%20Value :type Float%20%28double-precision%29) (Min :type Float%20%28double-precision%29) (Max%20-%20Min%20%28pre%20calculated%29 :type Float%20%28double-precision%29) (AudioTarget :type Audio%20Component%20Object%20Reference) (ParamName :type Name))
  ;[1 1 1] (input_val - min)/(max - min)
  (Audio|Parameter|SetFloatParameter AudioTarget ParamName (/ (- Input%20Value Min) Max%20-%20Min%20%28pre%20calculated%29) :!class /Script/AudioExtensions.AudioParameterControllerInterface)
  (return))

(fn ConstructionScript)
